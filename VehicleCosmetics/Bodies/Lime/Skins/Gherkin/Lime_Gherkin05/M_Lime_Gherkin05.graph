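# Surface material for the Lime vehicle body skin Gherkin05, output through the shared CharacterDeRez_Transition function.
# The skin mask's green channel blends in power-adjusted base color, higher metallic, lower roughness and the skin normal map.
# Outside that mask, the normal comes from a parameterized detail chain with animated UVs and optional body normal.
# The large decal network only drives the Detail1-to-Detail2 normal blend, which is off by default.
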Material /VehicleCosmetics/Bodies/Lime/Skins/Gherkin/Lime_Gherkin05/M_Lime_Gherkin05
graph {
 properties {
  bUseMaterialAttributes=true
  bUsedWithMeshDeformer=true
  bUsedWithSkeletalMesh=true
  pixelDepthOffsetMode=PDOM_Legacy
  usedShadingModels=Default Lit
 }
 nodes [
  n1 = TextureSample() {texture=&/Skin_Lime_Gherkin05_BC}
  n2 = Power(Base=@n1.RGB) {constExponent=0.7}
  n3 = TextureSample() {texture=&/Skin_Lime_Gherkin_05_RGB}
  n4 = Constant()
  n5 = Constant() {r=0.9}
  n6 = Constant() {r=0.5}
  n7 = Constant() {r=0.35}
  n8 = Constant3Vector() {constant=<rgba 0 0 1 0>}
  n9 = Time() {bOverride_Period=true, period=20}
  n10 = VectorParameter() {desc="(R,G) = Detail 1 Pan (X,Y)    (B,A) = Detail 2 Pan (X,Y)", group=Detail, parameterName=DetailTransformPanner}
  n11 = AppendVector(A=@n10.R, B=@n10.R)
  n12 = StaticSwitchParameter(True=TextureCoordinate() {coordinateIndex=1, uTiling=6, vTiling=6}, False=TextureCoordinate() {uTiling=6, vTiling=6}) {defaultValue=true, group=Detail, parameterName=Detail1UseCylinderUV}
  n13 = AppendVector(A=ComponentMask(None=@n12) {g=true}, B=ComponentMask(None=@n12) {r=true})
  n14 = VectorParameter() {desc="(R,G) = Detail1    (B,A) = Detail2", group=Detail, parameterName=DetailTransform, defaultValue=<rgba 1 1 1 1>}
  n15 = Rotator(
   Coordinate=Multiply(
    A=StaticSwitchParameter(True=@n13, False=@n12) {defaultValue=true, group=Detail, parameterName=Detail1SwapXY}
    B=ComponentMask(None=@n14.RGB) {g=true, r=true}
   )
   Time=ScalarParameter() {desc="Rotate Detail1 (0.25 = 90, 0.5 = 180, 0.75 = 270, 1=360)", group=Detail, parameterName=Detail1Rotation}
  ) {speed=6.2832}
  n16 = Time() {bOverride_Period=true, period=20}
  n17 = Multiply(
   A=@n16
   B=ScalarParameter() {desc="Detail1 Animated Spin Speed (Pivot point isn't very predictable)", group=Detail, parameterName=Detail1Spinner}
  )
  n18 = TextureSampleParameter2D(UVs=Add(A=Multiply(A=@n9, B=@n11), B=Rotator(Coordinate=@n15, Time=@n17) {speed=6.2832}) {desc=Panner}) {group=Detail, parameterName=Detail1, samplerSource=SSM_Wrap_WorldGroupSettings, texture=/VehicleCosmetics/Textures/T_Black}
  n19 = LinearInterpolate(A=@n8, B=@n18.RGB, Alpha=Saturate(None=ScalarParameter() {defaultValue=0.25, group=Detail, parameterName=Detail1Strength}))
  n20 = AppendVector(A=@n10.B, B=@n10.B)
  n21 = StaticSwitchParameter(True=TextureCoordinate() {coordinateIndex=1, uTiling=6, vTiling=6}, False=TextureCoordinate() {uTiling=6, vTiling=6}) {defaultValue=true, group=Detail, parameterName=Detail2UseCylinderUV}
  n22 = AppendVector(A=ComponentMask(None=@n21) {g=true}, B=ComponentMask(None=@n21) {r=true})
  n23 = Rotator(
   Coordinate=Multiply(
    A=StaticSwitchParameter(True=@n22, False=@n21) {defaultValue=true, group=Detail, parameterName=Detail2SwapXY}
    B=AppendVector(A=@n14.B, B=@n14.B)
   )
   Time=ScalarParameter() {desc="Rotate Detail2 (0.25 = 90, 0.5 = 180, 0.75 = 270, 1=360)", group=Detail, parameterName=Detail2Rotation}
  ) {speed=6.2832}
  n24 = Multiply(
   A=@n16
   B=ScalarParameter() {desc="Detail2 Animated Spin Speed (Pivot point isn't very predictable)", group=Detail, parameterName=Detail2Spinner}
  )
  n25 = TextureSampleParameter2D(UVs=Add(A=Multiply(A=@n9, B=@n20), B=Rotator(Coordinate=@n23, Time=@n24) {speed=6.2832}) {desc=Panner}) {group=Detail, parameterName=Detail2, samplerSource=SSM_Wrap_WorldGroupSettings, texture=&T_Black}
  n26 = LinearInterpolate(A=@n8, B=@n25.RGB, Alpha=Saturate(None=ScalarParameter() {defaultValue=0.25, group=Detail, parameterName=Detail2Strength}))
  n27 = StaticSwitchParameter(True=TextureCoordinate() {coordinateIndex=1}, False=TextureCoordinate()) {defaultValue=true, group=Decals, parameterName=*DecalsUseCylinderUV}
  n28 = VectorParameter() {desc="(R,G) = Tile (X,Y)    (B,A) = Pan (X,Y)", group=Decals, parameterName=Decal2Transform, defaultValue=<rgba 1 1 0 0>}
  n29 = ComponentMask(None=@n28.RGB) {g=true, r=true}
  n30 = Time() {bOverride_Period=true, period=20}
  n31 = VectorParameter() {desc="(R,G) = Decal 1 Pan (X,Y)    (B,A) = Decal 2 Pan (X,Y)", group=Decals, parameterName=DecalBases1*2Panner}
  n32 = Add(A=Add(A=Multiply(A=@n27, B=@n29), B=AppendVector(A=@n28.B, B=@n28.B)), B=Multiply(A=@n30, B=AppendVector(A=@n31.B, B=@n31.B))) {desc=Panner}
  n33 = ScalarParameter() {group=Decals, parameterName=Decal2Rotation}
  n34 = Time() {bOverride_Period=true, period=20}
  n35 = Rotator(
   Coordinate=StaticSwitchParameter(True=Rotator(Coordinate=@n32, Time=@n33) {speed=6.2832}, False=@n32) {desc=Enables Decal Sharpness and Threshold params., group=Decals, parameterName=Decal2Rotate}
   Time=Multiply(A=@n34, B=ScalarParameter() {desc="Decal2 Animated Spin Speed (Pivot point isn't very predictable)", group=Decals, parameterName=Decal2Spin})
  ) {speed=6.2832}
  n36 = ScalarParameter() {defaultValue=96, group=Pixelate, parameterName=NumPixels}
  n37 = TextureSampleParameter2D(UVs=StaticSwitchParameter(
    True=MaterialFunctionCall!(UV=@n35, NumPixels=@n36) {materialFunction=/VehicleCosmetics/Materials/ExtraFunctions/PixelateUV_MF}.Result
    False=@n35
   ) {group=Pixelate, parameterName=*PixelateDecal2}) {group=Decals, parameterName=Decal2, texture=&T_Black}
  n38 = StaticComponentMaskParameter(None=StaticSwitchParameter(True=OneMinus(None=@n37.RGB), False=@n37.RGB) {group=Decals, parameterName=Decal2Invert}) {defaultR=true, group=Decals, parameterName=Decal2Channel}
  n39 = Constant() {r=1}
  n40 = Add(
   A=Subtract(A=@n38, B=@n39)
   B=ScalarParameter() {defaultValue=0.5, desc=Tied to HeightMask switch., group=Decals, parameterName=Decal2Threshold}
  )
  n41 = ScalarParameter() {defaultValue=4, desc=Tied to HeightMask switch., group=Decals, parameterName=Decal2Sharpness}
  n42 = StaticSwitchParameter(True=Saturate(None=Multiply(A=@n40, B=@n41)), False=@n38) {desc=Enables Decal Sharpness and Threshold params., group=Decals, parameterName=Decal2HeightMask}
  n43 = TextureSampleParameter2D() {group=BaseTextures, parameterName=BodyTextures, samplerSource=SSM_Wrap_WorldGroupSettings, texture=&T_Black}
  n44 = StaticSwitchParameter(True=TextureCoordinate() {coordinateIndex=1, desc="UV1 (Cylinder)"}, False=TextureCoordinate() {desc="UV0 (Body Unwrap)"}) {defaultValue=true, group=Decals, parameterName=DecalBlendNoise*UseCylinderUV}
  n45 = StaticSwitchParameter(
   True=MaterialFunctionCall(UV=@n44, NumPixels=ScalarParameter() {defaultValue=96, group=Pixelate, parameterName=NumPixels}) {materialFunction=&PixelateUV_MF}.Result
   False=@n44
  ) {group=Pixelate, parameterName=*PixelateDecalBlendNoise}
  n46 = AppendVector(A=ComponentMask(None=@n45) {g=true}, B=ComponentMask(None=@n45) {r=true})
  n47 = ScalarParameter() {group=Decals, parameterName=DecalBlendRampRotation}
  n48 = StaticSwitchParameter(True=LinearInterpolate(A=@n46, B=@n45, Alpha=@n47), False=Rotator(Coordinate=@n46, Time=@n47) {speed=1}) {defaultValue=true, group=Decals, parameterName=DecalBlendRampCheapRotation}
  n49 = StaticSwitchParameter(True=@n43.R, False=ComponentMask(None=@n48) {r=true}) {desc=Instead of a linear ramp., group=Decals, parameterName=DecalBlend*UseEdgeDist}
  n50 = Constant() {r=1}
  n51 = Divide(A=Constant() {r=1}, B=ScalarParameter() {defaultValue=5, group=Decals, parameterName=DecalBlendRampWidth}) {desc=Invert multiplier}
  n52 = Divide(A=@n50, B=@n51) {desc=Get multiplier}
  n53 = Subtract(A=Constant() {r=-1}, B=@n52)
  n54 = ScalarParameter() {defaultValue=0.5, group=Decals, parameterName=DecalBlendRampPosition}
  n55 = ConstantBiasScale(None=@n54) {bias=0, scale=2}
  n56 = Add(A=@n52, B=@n50)
  n57 = Fmod(A=Multiply(A=Multiply(A=@n55, B=@n56), B=Constant() {r=0.5}), B=@n56)
  n58 = Multiply(A=Add(A=Add(A=@n49, B=@n53), B=Add(A=@n57, B=@n52)), B=@n51)
  n59 = VectorParameter() {desc="(R,G) = Tile (X,Y)    (B,A) = Pan (X,Y)", group=Decals, parameterName=DecalBlendNoiseTransform, defaultValue=<rgba 2 2 0 0>}
  n60 = AppendVector(A=@n59.B, B=@n59.B)
  n61 = StaticSwitchParameter(True=Multiply(A=@n60, B=Time() {bOverride_Period=true, period=20}), False=@n60) {desc=DecalBlendNoiseTransform vector parameter Blue and Alpha control pan speed., group=Decals, parameterName=DecalBlendNoiseDynamic}
  n62 = Add(A=Multiply(A=@n48, B=ComponentMask(None=@n59.RGB) {g=true, r=true}), B=@n61) {desc=Panner}
  n63 = Multiply(
   A=ComponentMask(None=StaticSwitchParameter(True=@n25.RGB, False=@n18.RGB) {group=Decals, parameterName=DecalBlendDistortUseDetail2}) {g=true, r=true}
   B=ScalarParameter() {defaultValue=0.25, group=Decals, parameterName=DecalBlendDistortStrength}
  )
  n64 = Rotator(
   Coordinate=StaticSwitchParameter(True=Add(A=@n62, B=@n63), False=@n62) {group=Decals, parameterName=DecalBlendDistortByDetail}
   Time=ScalarParameter() {desc="Rotate Decal Blend Noise (0.25 = 90, 0.5 = 180, 0.75 = 270, 1=360)", group=Decals, parameterName=DecalBlendNoiseRotation}
  ) {speed=6.2832}
  n65 = Multiply(
   A=Time() {bOverride_Period=true, period=20}
   B=ScalarParameter() {desc="DecalBlendNoise Animated Spin Speed (Pivot point isn't very predictable)", group=Decals, parameterName=DecalBlendNoiseSpin}
  )
  n66 = StaticComponentMaskParameter(None=TextureSampleParameter2D(UVs=Rotator(Coordinate=@n64, Time=@n65) {speed=6.2832}) {group=Decals, parameterName=DecalBlendNoise, texture=&T_Black}.RGB) {defaultR=true, desc=Uses FXNoise2, group=Decals, parameterName=DecalBlendNoiseChannel}
  n67 = OneMinus(None=@n66)
  n68 = Constant()
  n69 = Add(
   A=Subtract(
    A=StaticSwitchParameter(True=@n58, False=@n54) {defaultValue=true, desc=Instead of any ramp., group=Decals, parameterName=DecalBlend*UseRamp}
    B=Constant() {r=1}
   )
   B=StaticSwitchParameter(True=StaticSwitchParameter(True=@n67, False=@n66) {group=Decals, parameterName=DecalBlendNoiseInvert}, False=@n68) {defaultValue=true, group=Decals, parameterName=DecalBlend*UseNoise}
  )
  n70 = StaticSwitchParameter(True=@n42, False=@n68) {group=Decals, parameterName=DecalBlend*UseDecal2Height}
  n71 = Saturate(None=Multiply(A=Add(A=@n69, B=@n70), B=ScalarParameter() {defaultValue=32, group=Decals, parameterName=DecalBlendRampNoiseSharpness}))
  n72 = VectorParameter() {desc="(R,G) = Tile (X,Y)    (B,A) = Pan (X,Y)", group=Decals, parameterName=Decal1Transform, defaultValue=<rgba 1 1 0 0>}
  n73 = ComponentMask(None=@n72.RGB) {g=true, r=true}
  n74 = Add(A=Add(A=Multiply(A=@n27, B=@n73), B=AppendVector(A=@n72.B, B=@n72.B)), B=Multiply(A=@n30, B=AppendVector(A=@n31.R, B=@n31.R))) {desc=Panner}
  n75 = ScalarParameter() {group=Decals, parameterName=Decal1Rotation}
  n76 = Rotator(
   Coordinate=StaticSwitchParameter(True=Rotator(Coordinate=@n74, Time=@n75) {speed=6.2832}, False=@n74) {desc=Enables Decal Sharpness and Threshold params., group=Decals, parameterName=Decal1Rotate}
   Time=Multiply(
    A=@n34
    B=ScalarParameter() {desc="Decal 1 Animated Spin Speed (Pivot point isn't very predictable)", group=Decals, parameterName=Decal1Spin}
   )
  ) {speed=6.2832}
  n77 = TextureSampleParameter2D(UVs=StaticSwitchParameter(
    True=MaterialFunctionCall(UV=@n76, NumPixels=@n36) {materialFunction=&PixelateUV_MF}.Result
    False=@n76
   ) {group=Pixelate, parameterName=*PixelateDecal1}) {group=Decals, parameterName=Decal1, texture=&T_Black}
  n78 = StaticComponentMaskParameter(None=StaticSwitchParameter(True=OneMinus(None=@n77.RGB), False=@n77.RGB) {group=Decals, parameterName=Decal1Invert}) {defaultR=true, group=Decals, parameterName=Decal1Channel}
  n79 = Add(
   A=Subtract(A=@n78, B=@n39)
   B=ScalarParameter() {defaultValue=0.5, desc=Tied to HeightMask switch., group=Decals, parameterName=Decal1Threshold}
  )
  n80 = ScalarParameter() {defaultValue=4, desc=Tied to HeightMask switch., group=Decals, parameterName=Decal1Sharpness}
  n81 = StaticSwitchParameter(True=Saturate(None=Multiply(A=@n79, B=@n80)), False=@n78) {desc=Enables Decal Sharpness and Threshold params., group=Decals, parameterName=Decal1HeightMask}
  n82 = StaticSwitchParameter(
   True=StaticSwitchParameter(True=LinearInterpolate(A=@n81, B=@n42, Alpha=@n71), False=@n81) {defaultValue=true, group=Decals, parameterName=*DecalLayer2}
   False=Constant()
  ) {defaultValue=true, group=Decals, parameterName=*DecalLayer1}
  n83 = StaticSwitchParameter(True=StaticSwitchParameter(True=LinearInterpolate(A=@n19, B=@n26, Alpha=@n82), False=@n19) {group=Detail, parameterName=*Detail2}, False=@n8) {defaultValue=true, group=Detail, parameterName=*Detail}
  n84 = Constant3Vector() {constant=<rgba 1 1 0 0>}
  n85 = TextureSampleParameter2D() {
   group=BaseTextures
   parameterName=Ubershader_Normal
   samplerSource=SSM_Wrap_WorldGroupSettings
   samplerType=SAMPLERTYPE_LinearColor
   texture=/VehicleCosmetics/Textures/T_Black_Linear
  }
  n86 = AppendVector(A=LinearInterpolate(Alpha=@n85.R) {constA=-1, desc=R}, B=LinearInterpolate(Alpha=@n85.G) {constA=-1, desc=G}) {desc=R+G}
  n87 = Constant() {desc=Original BodyN B, r=1}
  n88 = StaticSwitchParameter(True=Add(A=Multiply(A=@n83, B=@n84), B=AppendVector(A=@n86, B=@n87) {desc=R+G+B}), False=@n83) {group=Lighting, parameterName=*LightUseBodyNormal}
  n89 = TextureSample() {samplerType=SAMPLERTYPE_Normal, texture=&/Skin_Lime_Gherkin_05_N}
  MaterialFunctionCall!(
   Fresnel
   Material=MakeMaterialAttributes!(
    BaseColor=LinearInterpolate(A=@n1.RGB, B=@n2, Alpha=@n3.G)
    Metallic=LinearInterpolate(A=@n4, B=@n5, Alpha=@n3.G)
    Specular=Constant() {r=16}
    Roughness=LinearInterpolate(A=@n6, B=@n7, Alpha=@n3.G)
    Anisotropy
    EmissiveColor
    Opacity
    OpacityMask
    Normal=LinearInterpolate(A=@n88, B=@n89.RGB, Alpha=@n3.G)
    Tangent
    WorldPositionOffset
    SubsurfaceColor
    ClearCoat
    ClearCoatRoughness
    AmbientOcclusion
    Refraction
    CustomizedUVs_0
    CustomizedUVs_1
    CustomizedUVs_2
    CustomizedUVs_3
    CustomizedUVs_4
    CustomizedUVs_5
    CustomizedUVs_6
    CustomizedUVs_7
    PixelDepthOffset
    ShadingModel
    Displacement
   )
  ) {materialFunction=/Game/Packages/Fortress_SharedMaterials/Base_Material_Functions/CharacterDeRez_Transition}
 ]
}
